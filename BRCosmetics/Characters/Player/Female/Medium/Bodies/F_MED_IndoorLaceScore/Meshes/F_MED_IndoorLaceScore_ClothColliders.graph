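# Physics asset of cloth colliders for the IndoorLaceScore female medium player body cosmetic, by its name meant as collision for simulated cloth.
# Capsules on pelvis, spine_01 and spine_03 approximate the torso.
# Thigh and calf bodies exist on both sides but list no shapes.
# The only constraint links spine_03 to spine_01 with free swing and twist, so its angular limit values do not restrict motion.
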PhysicsAsset /BRCosmetics/Characters/Player/Female/Medium/Bodies/F_MED_IndoorLaceScore/Meshes/F_MED_IndoorLaceScore_ClothColliders
bodies {
 calf_l []
 calf_r []
 pelvis [{shapeType=Capsule, length=7.078, radius=11.4298, center=<-3.1171 -0.8688 -0.0737>}]
 spine_01 [{shapeType=Capsule, length=7.3994, radius=7.6784, shapeName=spine_01_capsule, center=<2.5068 0.8204 0>}]
 spine_03 [{shapeType=Capsule, length=11.965, radius=11.4288, shapeName=spine_01_capsule, center=<4.821 2.9014 0>}]
 thigh_l []
 thigh_r []
}
constraints [{
  bone1Name=spine_03
  bone2Name=spine_01
  swing1LimitDegrees=45
  swing1Motion=Free
  swing2LimitDegrees=45
  swing2Motion=Free
  twistLimitDegrees=45
  twistMotion=Free
 }]
