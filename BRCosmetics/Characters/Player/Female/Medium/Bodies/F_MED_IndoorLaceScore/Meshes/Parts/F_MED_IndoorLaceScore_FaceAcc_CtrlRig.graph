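# Control Rig for the F_MED_IndoorLaceScore face accessory that procedurally adjusts hair bones to follow body pose.
# Head-versus-neck rotations become up, overshoot and down weights that shift spine-parented targets, which aim constraints use to orient back and front-left hair.
# Corrective offsets add head tilt, spherical arm/clavicle pose readers and targeting state.
# A backpack-cape flag cancels the arm-driven right-hair corrective; the clavicle corrective's interpolation input is fixed at zero.
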ControlRigBlueprint /BRCosmetics/Characters/Player/Female/Medium/Bodies/F_MED_IndoorLaceScore/Meshes/Parts/F_MED_IndoorLaceScore_FaceAcc_CtrlRig
graphs [
 CollapseNode_ContainedGraph {nodes [
   n1 = RigVMFunctionEntryNode()
   n2 = &RigVMFunction_Sequence(ExecuteContext=@n1.ExecuteContext)
   RigVMFunction_Sequence_2 = RigVMAggregateNode(ExecuteContext=@n2.A)
   Delta Rotation = RigVMFunctionReferenceNode(Child=head, Parent=neck_02, Source Maximum=-50, Source Minimum=-15, ExecuteContext=@RigVMFunction_Sequence_2.A)
   > RigVMVariableNode(Value=@"Delta Rotation".Delta) {variable=Head Up}
   > Delta Rotation_2 = RigVMFunctionReferenceNode(Child=head, Parent=neck_02, Source Maximum=-60, Source Minimum=-52)
   > RigVMVariableNode(Value=@"Delta Rotation_2".Delta) {variable=Head Up OverShoot}
   > Delta Rotation_1 = RigVMFunctionReferenceNode(Child=head, Parent=neck_02, Source Maximum=50, Source Minimum=-10)
   > RigVMVariableNode(Value=@"Delta Rotation_1".Delta) {variable=Head Down}
   &RigUnit_ModifyTransforms(
    ItemToModify=((Item=(Type=Bone,Name="head_aim_tgt"),Transform=(Rotation=(X=0,Y=0,Z=0,W=1),Translation=(X=-10,Y=5,Z=0),Scale3D=(X=1,Y=1,Z=1))))
    WeightMinimum=0
    ExecutePin=@RigVMFunction_Sequence_2.B
    Weight=RigVMVariableNode() {variable=Head Up}.Value
   )
   > &RigUnit_ModifyTransforms(
    ItemToModify=((Item=(Type=Bone,Name="head_aim_tgt"),Transform=(Rotation=(X=0,Y=0,Z=0,W=1),Translation=(X=-10,Y=-5,Z=0),Scale3D=(X=1,Y=1,Z=1))))
    WeightMinimum=0
    Weight=RigVMVariableNode() {variable=Head Up OverShoot}.Value
   )
   > &RigUnit_ModifyTransforms(
    ItemToModify=((Item=(Type=Bone,Name="head_aim_tgt"),Transform=(Rotation=(X=0,Y=0,Z=0,W=1),Translation=(X=0,Y=-20,Z=0),Scale3D=(X=1,Y=1,Z=1))))
    WeightMinimum=0
    Weight=RigVMVariableNode() {variable=Head Down}.Value
   )
   > DeltaRotation = RigVMFunctionReferenceNode(Child=Head, Parent=neck_01, Source Max=-60, Source Min=-10)
   > &RigUnit_ModifyTransforms(
    ItemToModify=((Item=(Type=Bone,Name="head_aim_tgt"),Transform=(Rotation=(X=-0.0714,Y=0,Z=0,W=0.9974),Translation=(X=0,Y=-1.9816,Z=9.4473),Scale3D=(X=1,Y=1,Z=1))))
    WeightMinimum=0
    Weight=@DeltaRotation.Delta_X_Negative
   )
   &RigUnit_AimConstraintLocalSpaceOffset(
    Child=(Type=Bone,Name="aim_hair_lf")
    Filter=(bX=true,bY=true,bZ=true)
    Parents=((Item=(Type=Bone,Name="head_aim_tgt"),Weight=1))
    UpAxis=(X=0,Y=-1,Z=0)
    Weight=1
    WorldUp=(Target=(X=0,Y=-1,Z=0),Kind=Location,Space=(Type=Bone,Name="head_aim_tgt"))
    bMaintainOffset=true
    ExecutePin=@RigVMFunction_Sequence_2.D
   )
   > &RigUnit_AimConstraintLocalSpaceOffset(
    Child=(Type=Bone,Name="aim_hair_cr")
    Filter=(bX=true,bY=true,bZ=true)
    Parents=((Item=(Type=Bone,Name="head_aim_tgt"),Weight=1))
    UpAxis=(X=0,Y=-1,Z=0)
    Weight=1
    WorldUp=(Target=(X=0,Y=-1,Z=0),Kind=Location,Space=(Type=Bone,Name="head_aim_tgt"))
    bMaintainOffset=true
   )
   > &RigUnit_AimConstraintLocalSpaceOffset(
    Child=(Type=Bone,Name="aim_hair_rt")
    Filter=(bX=true,bY=true,bZ=true)
    Parents=((Item=(Type=Bone,Name="head_aim_tgt"),Weight=1))
    UpAxis=(X=0,Y=-1,Z=0)
    Weight=1
    WorldUp=(Target=(X=0,Y=-1,Z=0),Kind=Location,Space=(Type=Bone,Name="head_aim_tgt"))
    bMaintainOffset=true
   )
   &RigUnit_ModifyTransforms(
    ItemToModify=((Item=(Type=Bone,Name="base_hair_rt"),Transform=(Rotation=(X=0,Y=0,Z=0,W=1),Translation=(X=-5,Y=0,Z=0),Scale3D=(X=1,Y=1,Z=1))))
    WeightMinimum=0
    ExecutePin=@RigVMFunction_Sequence_2.E
    Weight=RigVMVariableNode() {variable=Head Up}.Value
   )
   > Delta Rotation_3 = RigVMFunctionReferenceNode(Child=head, Parent=neck_02, Source Maximum=-21, Source Minimum=20)
   > n3 = &RigUnit_SphericalPoseReader(
    ActiveRegionScaleFactors=(PositiveWidth=1,NegativeWidth=1,PositiveHeight=1,NegativeHeight=1)
    ActiveRegionSize=0.4
    Debug=(bDrawDebug=false,bDraw2D=False,bDrawLocalAxes=False,DebugScale=25,DebugSegments=20,DebugThickness=0.25)
    DriverAxis=(X=-1,Y=0,Z=0)
    DriverItem=(Type=Bone,Name="upperarm_r")
    FalloffRegionScaleFactors=(PositiveWidth=1,NegativeWidth=1,PositiveHeight=1,NegativeHeight=1)
    FalloffSize=0.5
    FlipHeightScaling=False
    FlipWidthScaling=False
    OptionalParentItem=(Type=Bone,Name="None")
    RotationOffset=(X=-30,Y=40,Z=40)
   )
   n4 = RigVMVariableNode() {variable=isWearingBackPackCape}
   n5 = &RigVMFunction_AlphaInterp(
    InterpSpeedDecreasing=10
    InterpSpeedIncreasing=10
    Scale=1
    bClampResult=false
    bMapRange=false
    Value=&RigVMFunction_MathBoolToFloat(Value=RigVMVariableNode() {variable=isTargeting}.Value).Result
   )
   n6 = &RigVMFunction_AnimEvalRichCurve(
    Curve="(EditorCurveData=(Keys=((InterpMode=RCIM_Cubic,TangentMode=RCTM_Auto,TangentWeightMode=RCTWM_WeightedNone,Time=0,Value=0,ArriveTangent=0,ArriveTangentWeight=0,LeaveTangent=0,LeaveTangentWeight=0),(InterpMode=RCIM_Cubic,TangentMode=RCTM_Auto,TangentWeightMo ...(truncated)"
    SourceMinimum=0
    TargetMinimum=0
    Value=@n5.Result
   )
   Multiply = RigVMAggregateNode(C=@n6.Result, B=@n3.OutputParam, A=@"Delta Rotation_3".Delta)
   > &RigUnit_ModifyTransforms(
    ItemToModify=((Item=(Type=Bone,Name="dyn_hair_02_rt_helper"),Transform=(Rotation=(X=0,Y=0.1736,Z=0,W=0.9848),Translation=(X=-5,Y=0,Z=2),Scale3D=(X=1,Y=1,Z=0.6))))
    WeightMinimum=0
    Weight=&RigVMFunction_MathDoubleRemap(
     SourceMaximum=1
     SourceMinimum=0
     TargetMaximum=0
     Value=&RigVMFunction_MathBoolToFloat(Value=@n4.Value).Result
     TargetMinimum=@Multiply.Result
    ).Result
   )
   Remap = &RigVMFunction_MathDoubleRemap(SourceMaximum=0.25, SourceMinimum=0, TargetMaximum=1, TargetMinimum=0, Value=RigVMVariableNode() {variable=Head Down}.Value)
   > &RigUnit_ModifyTransforms(
    ItemToModify=((Item=(Type=Bone,Name="dyn_hair_02_rt_helper"),Transform=(Rotation=(X=0,Y=0,Z=0,W=1),Translation=(X=0,Y=0,Z=1),Scale3D=(X=1,Y=1,Z=1))))
    WeightMinimum=0
    Weight=&RigVMFunction_MathDoubleMul(A=@Remap.Result, B=@n6.Result).Result
   )
   RigVMFunctionReturnNode(
    Result=(X=0,Y=0,Z=0,W=1)
    ExecuteContext=@n2.B
    Head Up=RigVMVariableNode() {variable=Head Up}.Value
    Head Up OverShoot=RigVMVariableNode() {variable=Head Up OverShoot}.Value
    Head Down=RigVMVariableNode() {variable=Head Down}.Value
   )
  ]}
 Multiply_1_ContainedGraph {nodes [
   n1 = RigVMFunctionEntryNode()
   Multiply = &RigVMFunction_MathDoubleMul(A=@n1.A, B=@n1.B)
   RigVMFunctionReturnNode(Result=&RigVMFunction_MathDoubleMul(B=@n1.C, A=@Multiply.Result).Result)
  ]}
 Multiply_2_ContainedGraph {nodes [
   n1 = RigVMFunctionEntryNode()
   Multiply = &RigVMFunction_MathDoubleMul(A=@n1.A, B=@n1.B)
   RigVMFunctionReturnNode(Result=&RigVMFunction_MathDoubleMul(B=@n1.C, A=@Multiply.Result).Result)
  ]}
 New Function_ContainedGraph {nodes [
   n1 = RigVMFunctionEntryNode()
   RigVMFunction_MathQuaternionToEuler = &RigVMFunction_MathQuaternionToEuler(
    RotationOrder=ZYX
    Value=&RigUnit_GetRelativeTransformForItem(
     Child=(Type=Bone,Name="hand_l")
     Parent=(Type=Bone,Name="lowerarm_l")
     bChildInitial=False
     bParentInitial=False
     Child.Name=@n1.Child
     Parent.Name=@n1.Parent
    ).RelativeTransform.Rotation
   )
   Remap = &RigVMFunction_MathDoubleRemap(
    TargetMaximum=1
    TargetMinimum=0
    Value=@RigVMFunction_MathQuaternionToEuler.Result.Z
    SourceMaximum=@n1.Source Maximum
    SourceMinimum=@n1.Source Minimum
   )
   > RigVMFunctionReturnNode(Delta=@Remap.Result)
  ]}
 New Function_ContainedGraph {nodes [
   n1 = RigVMFunctionEntryNode()
   RigVMFunction_MathQuaternionToEuler = &RigVMFunction_MathQuaternionToEuler(
    RotationOrder=ZYX
    Value=&RigUnit_GetRelativeTransformForItem(
     Child=(Type=Bone,Name="hand_l")
     Parent=(Type=Bone,Name="lowerarm_l")
     bChildInitial=false
     bParentInitial=false
     Child.Name=@n1.Child
     Parent.Name=@n1.Parent
    ).RelativeTransform.Rotation
   )
   Remap = &RigVMFunction_MathDoubleRemap(
    TargetMaximum=1
    TargetMinimum=0
    Value=@RigVMFunction_MathQuaternionToEuler.Result.Y
    SourceMaximum=@n1.Source Maximum
    SourceMinimum=@n1.Source Minimum
   )
   > RigVMFunctionReturnNode(Delta=@Remap.Result)
  ]}
 New Function_ContainedGraph {nodes [
   n1 = RigVMFunctionEntryNode()
   RigVMFunction_MathBoolToInteger = &RigVMFunction_MathBoolToInteger(Value=@n1.onlyTwist)
   RigVMFunction_MathQuaternionSwingTwist = &RigVMFunction_MathQuaternionSwingTwist(
    twistAxis=@n1.twistAxis
    Input=&RigUnit_GetRelativeTransformForItem(
     bChildInitial=False
     bParentInitial=False
     child=(Type=Bone,Name="None")
     parent=(Type=Bone,Name="None")
     child.Name=@n1.child
     parent.Name=@n1.parent
     parent.Type=@n1.parentType
    ).RelativeTransform.Rotation
   )
   RigVMFunction_MathQuaternionToEuler = &RigVMFunction_MathQuaternionToEuler(
    Value=RigVMDispatchNode(
     Values=((),())
     Index=@RigVMFunction_MathBoolToInteger.Result
     Values.0=@RigVMFunction_MathQuaternionSwingTwist.Swing
     Values.1=@RigVMFunction_MathQuaternionSwingTwist.Twist
    ) {template="DISPATCH_RigVMDispatch_SelectInt32(in Index,in Values,out Result)"}.Result
    RotationOrder=@n1.Rotation Order
   )
   > DISPATCH_RigVMDispatch_Print = RigVMDispatchNode(ScreenColor=(R=1,G=1,B=1,A=1), ScreenDuration=1.5, Enabled=@n1.debugRotation, Value=@RigVMFunction_MathQuaternionToEuler.Result) {template="DISPATCH_RigVMDispatch_Print(in Prefix,in Value,in Enabled,in ScreenDuration,in ScreenColor)"}
   Clamp = &RigVMFunction_MathIntClamp(Value=@n1.primAxis)
   DISPATCH_RigVMDispatch_SelectInt32 = RigVMDispatchNode(
    Values=(0,0,0)
    Index=@Clamp.Result
    Values.0=@RigVMFunction_MathQuaternionToEuler.Result.X
    Values.1=@RigVMFunction_MathQuaternionToEuler.Result.Y
    Values.2=@RigVMFunction_MathQuaternionToEuler.Result.Z
   ) {template="DISPATCH_RigVMDispatch_SelectInt32(in Index,in Values,out Result)"}
   Remap = &RigVMFunction_MathDoubleRemap(
    TargetMaximum=1
    TargetMinimum=0
    bClamp=False
    Value=@DISPATCH_RigVMDispatch_SelectInt32.Result
    SourceMaximum=@n1.max
    SourceMinimum=@n1.min
   )
   Multiply = &RigVMFunction_MathDoubleMul(A=&RigVMFunction_MathDoubleClamp(Value=@Remap.Result).Result, B=@n1.premult)
   > RigVMFunctionReturnNode(outputWeight=@Multiply.Result)
  ]}
 New Function_ContainedGraph {nodes [
   n1 = RigVMFunctionEntryNode()
   RigVMFunction_MathQuaternionToEuler = &RigVMFunction_MathQuaternionToEuler(
    RotationOrder=ZYX
    Value=&RigUnit_GetRelativeTransformForItem(
     Child=(Type=Bone,Name="head")
     Parent=(Type=Bone,Name="neck_01")
     bChildInitial=False
     bParentInitial=False
     Child.Name=@n1.Child
     Parent.Name=@n1.Parent
    ).RelativeTransform.Rotation
   )
   Remap = &RigVMFunction_MathDoubleRemap(
    TargetMaximum=1
    TargetMinimum=0
    SourceMaximum=@n1.Source Max
    SourceMinimum=@n1.Source Min
    Value=@RigVMFunction_MathQuaternionToEuler.Result.Z
   )
   > RigVMFunctionReturnNode(
    Delta_z=@Remap.Result
    Delta_X_Negative=&RigVMFunction_MathDoubleRemap(SourceMaximum=-60, SourceMinimum=0, TargetMaximum=1, TargetMinimum=0, Value=@RigVMFunction_MathQuaternionToEuler.Result.X).Result
    Delta_X_Positive=&RigVMFunction_MathDoubleRemap(SourceMaximum=70, SourceMinimum=0, TargetMaximum=1, TargetMinimum=0, Value=@RigVMFunction_MathQuaternionToEuler.Result.X).Result
   )
  ]}
 New Function_ContainedGraph {nodes [
   n1 = RigVMFunctionEntryNode()
   RigVMFunction_MathQuaternionToEuler = &RigVMFunction_MathQuaternionToEuler(
    RotationOrder=ZYX
    Value=&RigUnit_GetRelativeTransformForItem(
     Child=(Type=Bone,Name="head")
     Parent=(Type=Null,Name="neck_01")
     bChildInitial=False
     bParentInitial=False
     Child.Name=@n1.Child
     Parent.Name=@n1.Parent
    ).RelativeTransform.Rotation
   )
   Remap = &RigVMFunction_MathDoubleRemap(
    TargetMaximum=1
    TargetMinimum=0
    SourceMaximum=@n1.Source Max
    SourceMinimum=@n1.Source Min
    Value=@RigVMFunction_MathQuaternionToEuler.Result.Z
   )
   > RigVMFunctionReturnNode(
    Delta_z=@Remap.Result
    Delta_X_Negative=&RigVMFunction_MathDoubleRemap(SourceMaximum=-60, SourceMinimum=0, TargetMaximum=1, TargetMinimum=0, Value=@RigVMFunction_MathQuaternionToEuler.Result.X).Result
    Delta_X_Positive=&RigVMFunction_MathDoubleRemap(SourceMaximum=70, SourceMinimum=0, TargetMaximum=1, TargetMinimum=0, Value=@RigVMFunction_MathQuaternionToEuler.Result.X).Result
   )
  ]}
 RigVMFunctionLibrary {nodes [
   Delta Rotation = RigVMCollapseNode(ExecuteContext=())
   Delta Rotation_clavicle_up = RigVMCollapseNode(ExecuteContext=())
   DeltaRotation = RigVMCollapseNode(ExecuteContext=(), Rotation Order=ZYX, premult=1, twistAxis=(X=1,Y=0,Z=0))
   DeltaRotation_1 = RigVMCollapseNode(ExecuteContext=())
   DeltaRotation_2 = RigVMCollapseNode(ExecuteContext=())
  ]}
 RigVMFunction_Sequence_3_ContainedGraph {nodes [
   n1 = RigVMFunctionEntryNode()
   n2 = &RigVMFunction_Sequence(ExecuteContext=@n1.ExecuteContext)
   n3 = &RigVMFunction_Sequence(ExecuteContext=@n2.B)
   n4 = &RigVMFunction_Sequence(ExecuteContext=@n3.B)
   RigVMFunctionReturnNode(B=@n3.A, A=@n2.A, D=@n4.A, E=@n4.B)
  ]}
 RigVMModel {nodes [
   RigUnit_BeginExecution = &RigUnit_BeginExecution()
   "HairSetup " = RigVMCollapseNode(ExecuteContext=@RigUnit_BeginExecution.ExecutePin)
   > &RigUnit_ParentConstraint(
    AdvancedSettings=(InterpolationType=Shortest,RotationOrderForFilter=XYZ)
    Child=(Type=Bone,Name="base_hair_front_aim_tgt")
    Filter=(TranslationFilter=(bX=True,bY=True,bZ=True),RotationFilter=(bX=true,bY=true,bZ=true),ScaleFilter=(bX=True,bY=True,bZ=True))
    Parents=((Item=(Type=Bone,Name="spine_05"),Weight=1),(Item=(Type=Bone,Name="clavicle_l"),Weight=1))
    Weight=1
   )
   > &RigUnit_ModifyTransforms(
    ItemToModify=((Item=(Type=Bone,Name="hair_front_aim_tgt"),Transform=(Rotation=(X=0,Y=0,Z=0,W=1),Translation=(X=0,Y=0,Z=20),Scale3D=(X=1,Y=1,Z=1))))
    WeightMinimum=0
    Weight=@"HairSetup ".Head Up
   )
   > &RigUnit_AimConstraintLocalSpaceOffset(
    Child=(Type=Bone,Name="aim_hair_front_lf")
    Filter=(bX=true,bY=true,bZ=true)
    Parents=((Item=(Type=Bone,Name="hair_front_aim_tgt"),Weight=1))
    Weight=1
    WorldUp=(Target=(X=0,Y=0,Z=1),Kind=Location,Space=(Type=Bone,Name="hair_front_aim_tgt"))
   )
   > DeltaRotation_2 = RigVMFunctionReferenceNode(Child=head, Parent=head_null)
   > &RigUnit_ModifyTransforms(
    ItemToModify=((Item=(Type=Bone,Name="offset_hair_front_lf"),Transform=(Rotation=(X=-0.5375,Y=0.0541,Z=-0.0019,W=-0.8415),Translation=(X=-2.9005,Y=0.4992,Z=0.334),Scale3D=(X=1,Y=1,Z=1))))
    WeightMinimum=0
    Weight=@DeltaRotation_2.Delta_X_Negative
   )
   > n1 = &RigUnit_SphericalPoseReader(
    ActiveRegionScaleFactors=(PositiveWidth=1,NegativeWidth=1,PositiveHeight=1,NegativeHeight=1)
    ActiveRegionSize=0.15
    Debug=(bDrawDebug=false,bDraw2D=False,bDrawLocalAxes=False,DebugScale=25,DebugSegments=20,DebugThickness=0.25)
    DriverAxis=(X=1,Y=0,Z=0)
    DriverItem=(Type=Bone,Name="clavicle_l")
    FalloffRegionScaleFactors=(PositiveWidth=1,NegativeWidth=1,PositiveHeight=1,NegativeHeight=1)
    FalloffSize=0.25
    FlipHeightScaling=False
    FlipWidthScaling=False
    OptionalParentItem=(Type=Bone,Name="None")
    RotationOffset=(X=50,Y=0,Z=-90)
   )
   RigVMFunction_MathBoolToFloat = &RigVMFunction_MathBoolToFloat(Value=RigVMVariableNode() {variable=isTargeting}.Value)
   AlphaInterp = &RigVMFunction_AlphaInterp(InterpSpeedDecreasing=10, InterpSpeedIncreasing=10, Value=0, bClampResult=False, Scale=@RigVMFunction_MathBoolToFloat.Result)
   Remap = &RigVMFunction_MathDoubleRemap(SourceMaximum=0.5, SourceMinimum=0, TargetMaximum=1, TargetMinimum=0, Value=@"HairSetup ".Head Up)
   Multiply = RigVMAggregateNode(C=@AlphaInterp.Result, A=@Remap.Result, B=@n1.OutputParam)
   > &RigUnit_ModifyTransforms(
    ItemToModify=((Item=(Type=Bone,Name="base_hair_front_01_lf"),Transform=(Rotation=(X=0.278,Y=0.2092,Z=-0.0965,W=0.9325),Translation=(X=0.5823,Y=-1.789,Z=2.6924),Scale3D=(X=1,Y=1,Z=1))))
    WeightMinimum=0
    Weight=@Multiply.Result
   )
  ]}
]
hierarchy [
 root {loc=<0 0 0>, rot=<0 0 0 1>}
 pelvis {parent=root, loc=<0 -0.5301 98.0571>, rot=<0.0127 0.707 -0.0127 -0.707>}
 spine_01 {parent=pelvis, loc=<6.1208 -0.9328 0>, rot=<0 0 0.034 -0.9994>}
 spine_02 {parent=spine_01, loc=<6.5997 -0.2095 0>, rot=<0 0 0.0238 -0.9997>}
 spine_03 {parent=spine_02, loc=<6.4914 -0.1237 0>, rot=<0 0 0.0149 -0.9999>}
 spine_04 {parent=spine_03, loc=<6.9598 0.0794 0>, rot=<0 0 -0.0132 -0.9999>}
 spine_05 {parent=spine_04, loc=<6.9987 0 0>, rot=<0 0 -0.0015 -1>}
 neck_01 {parent=spine_05, loc=<11.3138 0.1529 0>, rot=<0 0 -0.2003 -0.9797>}
 neck_02 {parent=neck_01, loc=<4.5627 0 0>, rot=<0 0 0.0504 -0.9987>}
 head {parent=neck_02, loc=<5.3914 0 0>, rot=<0 0 0.115 -0.9934>}
 base_hair_lf {parent=head, loc=<9.7714 -3.8 -5.9095>, rot=<0.0609 -0.6994 0.7097 -0.0581>}
 aim_hair_lf {parent=base_hair_lf, loc=<0 0 0>, rot=<0 0 0 1>}
 offset_hair_lf {parent=aim_hair_lf, loc=<0 0 0>, rot=<0 0 0 1>}
 dyn_hair_01_lf {parent=offset_hair_lf, loc=<0 0 0>, rot=<0 0 0 1>}
 dyn_hair_02_lf {parent=dyn_hair_01_lf, loc=<13.5092 0 0>, rot=<0 0.1499 0 -0.9887>}
 dyn_hair_03_lf {parent=dyn_hair_02_lf, loc=<7.0701 0 0>, rot=<0 -0.0342 0 -0.9994>}
 dyn_hair_04_lf {parent=dyn_hair_03_lf, loc=<7.0629 0 0>, rot=<0 -0.1708 0 -0.9853>}
 dyn_hair_05_lf {parent=dyn_hair_04_lf, loc=<7.0605 0 0>, rot=<0 -0.1054 0 -0.9944>}
 base_hair_cr {parent=head, loc=<10.136 -4.1709 -0.0414>, rot=<0.0722 -0.6931 0.7138 -0.0702>}
 aim_hair_cr {parent=base_hair_cr, loc=<0 0 0>, rot=<0 0 0 1>}
 offset_hair_cr {parent=aim_hair_cr, loc=<0 0 0>, rot=<0 0 0 1>}
 dyn_hair_01_cr {parent=offset_hair_cr, loc=<0 0 0>, rot=<0 0 0 1>}
 dyn_hair_02_cr {parent=dyn_hair_01_cr, loc=<13.7912 0 0>, rot=<0 0.106 0 -0.9944>}
 dyn_hair_03_cr {parent=dyn_hair_02_cr, loc=<7.0701 0 0>, rot=<0 0.0189 0 -0.9998>}
 dyn_hair_04_cr {parent=dyn_hair_03_cr, loc=<7.0629 0 0>, rot=<0 -0.2023 0 -0.9793>}
 dyn_hair_05_cr {parent=dyn_hair_04_cr, loc=<7.0605 0 0>, rot=<0 -0.1054 0 -0.9944>}
 base_hair_rt {parent=head, loc=<9.8034 -3.6268 5.791>, rot=<0.0609 -0.6994 0.7097 -0.0581>}
 aim_hair_rt {parent=base_hair_rt, loc=<0 0 0>, rot=<0 0 0 1>}
 offset_hair_rt {parent=aim_hair_rt, loc=<0 0 0>, rot=<0 0 0 1>}
 dyn_hair_01_rt {parent=offset_hair_rt, loc=<0 0 0>, rot=<0 0 0 1>}
 dyn_hair_02_rt {parent=dyn_hair_01_rt, loc=<13.5092 0 0>, rot=<0 0.1499 0 -0.9887>}
 dyn_hair_03_rt {parent=dyn_hair_02_rt, loc=<7.0701 0 0>, rot=<0 -0.0342 0 -0.9994>}
 dyn_hair_04_rt {parent=dyn_hair_03_rt, loc=<7.0629 0 0>, rot=<0 -0.1708 0 -0.9853>}
 dyn_hair_05_rt {parent=dyn_hair_04_rt, loc=<7.0605 0 0>, rot=<0 -0.1054 0 -0.9944>}
 base_hair_02_rt_helper {parent=dyn_hair_02_rt, loc=<4.0805 -3.1306 -2.7921>, rot=<0 0 0 1>}
 dyn_hair_02_rt_helper {parent=base_hair_02_rt_helper, loc=<0 0 0>, rot=<0 0.1499 0 -0.9887>}
 base_hair_front_01_lf {parent=head, loc=<2.702 1.7166 -7.7425>, rot=<0.0264 -0.7126 -0.7005 0.0291>}
 aim_hair_front_lf {parent=base_hair_front_01_lf, loc=<0 0 0>, rot=<0 0 0 1>}
 offset_hair_front_lf {parent=aim_hair_front_lf, loc=<0 0 0>, rot=<0 0 0 1>}
 dyn_hair_front_01_lf {parent=offset_hair_front_lf, loc=<0 0 0>, rot=<0 0 0 1>}
 dyn_hair_front_02_lf {parent=dyn_hair_front_01_lf, loc=<10.7465 -0.0083 0.1751>, rot=<0 0.2937 0 -0.9559>}
 dyn_hair_front_03_lf {parent=dyn_hair_front_02_lf, loc=<7.5611 0 0>, rot=<0 0 0 1>}
 head_null {parent=neck_02, type=Null, loc=<5.3914 0 0>, rot=<0 0 0.115 -0.9934>}
 clavicle_l {parent=spine_05, loc=<5.4831 -0.1053 -3.2946>, rot=<0.0465 -0.7289 -0.009 -0.6829>}
 upperarm_l {parent=clavicle_l, loc=<9.8658 -1.2136 -0.4105>, rot=<0.0164 -0.3245 -0.0269 -0.9454>}
 upperarm_l_null {parent=upperarm_l, type=Null, loc=<0 0 0>, rot=<0 0 0 1>}
 clavicle_l_null {parent=clavicle_l, type=Null, loc=<0 0 0>, rot=<0 0 0 1>}
 clavicle_r {parent=spine_05, loc=<5.4831 -0.1053 3.2946>, rot=<0.7289 0.0465 -0.6829 0.009>}
 upperarm_r {parent=clavicle_r, loc=<-9.8658 1.2136 0.4105>, rot=<0.0164 -0.3245 -0.0269 -0.9454>}
 upperarm_r_null {parent=upperarm_r, type=Null, loc=<0 0 0>, rot=<0 0 0 1>}
 clavicle_r_null {parent=clavicle_r, type=Null, loc=<0 0 0>, rot=<0 0 0 1>}
 base_head_aim_tgt {parent=spine_05, loc=<9.003 -16.7403 0.2354>, rot=<0.0075 -0.0008 0.0004 -1>}
 head_aim_tgt {parent=base_head_aim_tgt, loc=<0 0 0>, rot=<0 0 0 1>}
 base_hair_front_aim_tgt {parent=spine_05, loc=<-0.7283 12.698 -9.2939>, rot=<-0.156 -0.707 -0.639 -0.2599>}
 hair_front_aim_tgt {parent=base_hair_front_aim_tgt, loc=<0 0 0>, rot=<0 0 0 1>}
]
variables [
 Head Down {type=double, default=0}
 Head Up {type=double, default=0}
 Head Up OverShoot {type=double, default=0}
 IsSwingPickaxe {type=bool, default=False}
 isTargeting {type=bool, default=False}
 isWearingBackPackCape {type=bool, default=False}
]
