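# Physics asset for the face accessory part of the female medium IndoorLaceScore player cosmetic, set up for dynamic hair.
# Capsule bodies cover the head, neck, upper spine, clavicles and upper arms alongside the hair bones.
# Center, left and right five-bone strands and a three-bone left front strand chain from small offset-bone anchors.
# Twist is locked throughout, with tight root limits and generally wider swing toward the tips, while front joints bend on one axis.
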PhysicsAsset /BRCosmetics/Characters/Player/Female/Medium/Bodies/F_MED_IndoorLaceScore/Meshes/Parts/F_MED_IndoorLaceScore_FaceAcc_Physics
bodies {
 clavicle_l [{
   shapeType=Capsule
   length=8.1483
   radius=5.0631
   shapeName=clavicle_l_capsule
   center=<6.4618 -1.7066 -0.6916>
   rotation=<rot 81.344 -0.0676 179.9308>
  }]
 clavicle_r [{
   shapeType=Capsule
   length=7.4013
   radius=5.929
   shapeName=clavicle_r_capsule
   center=<-6.2251 1.0912 2.1208>
   rotation=<rot 87.5685 -179.9942 -179.9944>
  }]
 dyn_hair_01_cr [{shapeType=Capsule, length=6.9364, radius=2, shapeName=dyn_hair_01_rt_capsule, center=<5 0 0>, rotation=<rot 81.7452 0 0>}]
 dyn_hair_01_lf [{
   shapeType=Capsule
   length=6.834
   radius=3.673
   shapeName=dyn_hair_01_rt_capsule
   center=<5 3.0355 -2.6972>
   rotation=<rot 74.4491 179.9997 179.9998>
  }]
 dyn_hair_01_rt [{shapeType=Capsule, length=7.1145, radius=1.6222, shapeName=dyn_hair_01_rt_capsule, center=<5 0 0>, rotation=<rot 83.1702 0 0>}]
 dyn_hair_02_cr [{shapeType=Capsule, length=5.8123, radius=1.941, shapeName=dyn_hair_02_rt_capsule, center=<2 0 1.0253>, rotation=<rot 83.9553 180 180>}]
 dyn_hair_02_lf [{
   shapeType=Capsule
   length=2.9388
   radius=3.0193
   shapeName=dyn_hair_02_rt_capsule
   center=<5.2385 2.3311 -0.7579>
   rotation=<rot 55.1851 179.9994 179.9996>
  }]
 dyn_hair_02_rt [{
   shapeType=Capsule
   length=6.9627
   radius=1.6733
   shapeName=dyn_hair_02_rt_capsule
   center=<0.2825 0 0.6338>
   rotation=<rot 88.797 0.0061 0.0063>
  }]
 dyn_hair_03_cr [{shapeType=Capsule, length=5.9471, radius=2.0529, shapeName=dyn_hair_03_cr_capsule, center=<5.077 0 0.9926>, rotation=<rot 87.9195 0 0>}]
 dyn_hair_03_lf [{shapeType=Capsule, length=8.0885, radius=2.131, shapeName=dyn_hair_03_lf_capsule, center=<8.2769 5.218 -1.2451>, rotation=<rot 70.4787 0 0>}]
 dyn_hair_03_rt [
  {shapeType=Capsule, length=6.3594, radius=1.8512, shapeName=dyn_hair_03_rt_capsule, center=<3.1037 0 0.2256>, rotation=<rot 76.7576 180 180>}
  {shapeType=Capsule, length=3.58, radius=1.2808, shapeName=dyn_hair_03_rt_capsule, center=<6.0301 -5.1973 -1.3574>, rotation=<rot 87.545 0 0>}
 ]
 dyn_hair_04_cr [{shapeType=Capsule, length=8.7239, radius=2.1494, shapeName=dyn_hair_04_cr_capsule, center=<7.4287 0 0.8046>, rotation=<rot 88.5975 0 0>}]
 dyn_hair_04_lf [{
   shapeType=Capsule
   length=5.3061
   radius=2.0852
   shapeName=dyn_hair_04_lf_capsule
   center=<11.9747 4.0936 -0.6028>
   rotation=<rot 87.5939 0 0>
  }]
 dyn_hair_04_rt [{
   shapeType=Capsule
   length=9.0312
   radius=1.8421
   shapeName=dyn_hair_04_rt_capsule
   center=<7.3479 0 0.12>
   rotation=<rot 89.3926 0.0003 0.0003>
  }]
 dyn_hair_05_cr [{shapeType=Capsule, length=5, radius=2, shapeName=dyn_hair_05_cr_capsule, center=<5 0 1.2654>, rotation=<rot 79.4625 180 -180>}]
 dyn_hair_05_lf [{shapeType=Capsule, length=5.3098, radius=1.6902, shapeName=dyn_hair_05_lf_capsule, center=<4.9006 0 0.5958>, rotation=<rot 80.5288 180 180>}]
 dyn_hair_05_rt [{shapeType=Capsule, length=5.8607, radius=1.1393, shapeName=dyn_hair_05_rt_capsule, center=<4.8976 0 0.494>, rotation=<rot 78.2878 180 180>}]
 dyn_hair_front_01_lf [
  {shapeType=Capsule, length=8.728, radius=1.7438, shapeName=dyn_hair_front_01_lf_capsule, center=<2 0 3.503>, rotation=<rot 90 0 0>}
  {
   shapeType=Capsule
   length=7.3005
   radius=1.6338
   shapeName=dyn_hair_front_01_lf_capsule
   center=<9.3322 -1.4016 -2.4>
   rotation=<rot 86.425 0 0>
  }
 ]
 dyn_hair_front_02_lf [{
   shapeType=Capsule
   length=8.3588
   radius=1.6366
   shapeName=dyn_hair_front_02_lf_capsule
   center=<4.2317 -0.6163 -1.7181>
   rotation=<rot -69.4298 -13.4209 21.0401>
  }]
 dyn_hair_front_03_lf [{
   shapeType=Capsule
   length=6.7311
   radius=1.7875
   shapeName=dyn_hair_front_03_lf_capsule
   center=<4.3387 -0.339 -0.2988>
   rotation=<rot 78.5244 -179.9984 -180>
  }]
 head [{shapeType=Capsule, length=7.1907, radius=5.7787, shapeName=head_capsule, center=<0.9288 3.8536 -0.2703>, rotation=<rot 90 6.1106 0>}]
 neck_01 [{shapeType=Capsule, length=5.6236, radius=4.2828, shapeName=neck_01_capsule, center=<0.2313 0 -1.3905>, rotation=<rot 90 0 0>}]
 offset_hair_cr [{shapeType=Capsule, length=0.505, radius=0.505, shapeName=offset_hair_cr_capsule, rotation=<rot 0 -90 0>}]
 offset_hair_front_lf [{shapeType=Capsule, length=0.505, radius=0.505, shapeName=offset_hair_front_lf_capsule, rotation=<rot 0 -90 0>}]
 offset_hair_lf [{shapeType=Capsule, length=0.505, radius=0.505, shapeName=offset_hair_lf_capsule, rotation=<rot 0 -90 0>}]
 offset_hair_rt [{shapeType=Capsule, length=0.505, radius=0.505, shapeName=offset_hair_rt_capsule, rotation=<rot 0 -90 0>}]
 spine_04 [
  {shapeType=Capsule, length=16.0862, radius=8.9138, shapeName=spine_05_capsule, center=<-10.9301 -0.7195 0>, rotation=<rot 0 -90 0>}
  {shapeType=Capsule, length=17.2664, radius=8.9449, shapeName=spine_04_capsule, center=<0.8439 0.4891 0>}
  {shapeType=Capsule, length=14.5704, radius=8.0067, shapeName=spine_04_capsule, center=<0.1101 3.0149 0>}
 ]
 spine_05 [{shapeType=Capsule, length=17.9089, radius=7.5619, shapeName=spine_05_capsule, center=<-1.4739 -0.6664 0>, rotation=<rot 0 -90 0>}]
 upperarm_l [{
   shapeType=Capsule
   length=26.9049
   radius=5.4667
   shapeName=upperarm_l_capsule
   center=<12.95 -0.0617 -0.5337>
   rotation=<rot -0.0001 -89.9999 -92.7028>
  }]
 upperarm_r [{shapeType=Capsule, length=24.0453, radius=5.0231, shapeName=upperarm_l_capsule, center=<-13.9807 0.352 0.5342>, rotation=<rot 0 -90 -90>}]
}
constraints [
 {
  bone1Name=dyn_hair_01_cr
  bone2Name=offset_hair_cr
  swing1LimitDegrees=0
  swing1Motion=Locked
  swing2LimitDegrees=5
  twistLimitDegrees=0
  twistMotion=Locked
 }
 {bone1Name=dyn_hair_03_cr, bone2Name=dyn_hair_02_cr, swing1LimitDegrees=10, swing2LimitDegrees=20, twistLimitDegrees=0, twistMotion=Locked}
 {bone1Name=dyn_hair_04_cr, bone2Name=dyn_hair_03_cr, swing1LimitDegrees=10, swing2LimitDegrees=25, twistLimitDegrees=0, twistMotion=Locked}
 {bone1Name=dyn_hair_05_cr, bone2Name=dyn_hair_04_cr, swing1LimitDegrees=10, swing2LimitDegrees=30, twistLimitDegrees=45, twistMotion=Locked}
 {
  bone1Name=dyn_hair_01_lf
  bone2Name=offset_hair_lf
  swing1LimitDegrees=0
  swing1Motion=Locked
  swing2LimitDegrees=5
  twistLimitDegrees=0
  twistMotion=Locked
 }
 {bone1Name=dyn_hair_02_lf, bone2Name=dyn_hair_01_lf, swing1LimitDegrees=10, swing2LimitDegrees=20, twistLimitDegrees=0, twistMotion=Locked}
 {bone1Name=dyn_hair_03_lf, bone2Name=dyn_hair_02_lf, swing1LimitDegrees=10, swing2LimitDegrees=20, twistLimitDegrees=0, twistMotion=Locked}
 {bone1Name=dyn_hair_04_lf, bone2Name=dyn_hair_03_lf, swing1LimitDegrees=10, swing2LimitDegrees=25, twistLimitDegrees=0, twistMotion=Locked}
 {bone1Name=dyn_hair_05_lf, bone2Name=dyn_hair_04_lf, swing1LimitDegrees=10, swing2LimitDegrees=30, twistLimitDegrees=45, twistMotion=Locked}
 {
  bone1Name=dyn_hair_01_rt
  bone2Name=offset_hair_rt
  swing1LimitDegrees=0
  swing1Motion=Locked
  swing2LimitDegrees=5
  twistLimitDegrees=0
  twistMotion=Locked
 }
 {bone1Name=dyn_hair_02_rt, bone2Name=dyn_hair_01_rt, swing1LimitDegrees=10, swing2LimitDegrees=30, twistLimitDegrees=0, twistMotion=Locked}
 {bone1Name=dyn_hair_03_rt, bone2Name=dyn_hair_02_rt, swing1LimitDegrees=10, swing2LimitDegrees=20, twistLimitDegrees=0, twistMotion=Locked}
 {bone1Name=dyn_hair_04_rt, bone2Name=dyn_hair_03_rt, swing1LimitDegrees=10, swing2LimitDegrees=25, twistLimitDegrees=0, twistMotion=Locked}
 {bone1Name=dyn_hair_05_rt, bone2Name=dyn_hair_04_rt, swing1LimitDegrees=10, swing2LimitDegrees=30, twistLimitDegrees=45, twistMotion=Locked}
 {bone1Name=dyn_hair_02_cr, bone2Name=dyn_hair_01_cr, swing1LimitDegrees=10, swing2LimitDegrees=20, twistLimitDegrees=0, twistMotion=Locked}
 {
  bone1Name=dyn_hair_front_02_lf
  bone2Name=dyn_hair_front_01_lf
  swing1LimitDegrees=0
  swing1Motion=Locked
  swing2LimitDegrees=30
  twistLimitDegrees=0
  twistMotion=Locked
 }
 {
  bone1Name=dyn_hair_front_03_lf
  bone2Name=dyn_hair_front_02_lf
  swing1LimitDegrees=0
  swing1Motion=Locked
  swing2LimitDegrees=30
  twistLimitDegrees=0
  twistMotion=Locked
 }
 {
  bone1Name=dyn_hair_front_01_lf
  bone2Name=offset_hair_front_lf
  swing1LimitDegrees=0
  swing1Motion=Locked
  swing2LimitDegrees=30
  twistLimitDegrees=0
  twistMotion=Locked
 }
]
