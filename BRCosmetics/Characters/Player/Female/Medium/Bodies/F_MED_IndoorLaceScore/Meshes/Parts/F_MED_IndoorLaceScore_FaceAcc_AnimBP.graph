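# Post-process-style AnimBP for the IndoorLaceScore female face accessory, driving its hair bones.
# It blends body-mesh clavicle and upper arms onto the parent pose, runs a Control Rig fed melee, targeting and backbling flags, then simulates RigidBody physics relative to neck_02.
# Constraints to spine_05 and additive bone transforms then adjust hair bones.
# Backbling is checked on customization load; GravityOverride updates each frame, though RigidBody shows fixed gravity.
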AnimBlueprint /BRCosmetics/Characters/Player/Female/Medium/Bodies/F_MED_IndoorLaceScore/Meshes/Parts/F_MED_IndoorLaceScore_FaceAcc_AnimBP
target_skeleton=&/F_MED_IndoorLaceScore_FaceAcc_Skeleton
graphs [
 AnimGraph {nodes [
   Output Pose(Result=Use cached pose 'backpackCapeCache'()) {layerGroup=None}
   n1 = Layered blend per bone(
    BasePose=Copy Pose From Mesh() {bUseAttachedParent=true}
    BlendPoses_0=Copy Pose From Mesh(SourceMeshComponent=Get Body_Mesh()) {bUseAttachedParent=true}
   ) {layerSetup [{branchFilters [
       {boneName=clavicle_l}
       {boneName=clavicle_r}
       {boneName=upperarm_r}
       {boneName=upperarm_l}
       {boneName=deform_upperarm_bk_l}
       {boneName=deform_upperarm_bk_base_l}
       {boneName=upperarm_twist_01_l}
      ]}]}
   n2 = Control Rig!(
    Source=@n1
    Alpha
    IsSwingPickaxe:bool=Get bIsPlayingMeleeAnim()
    isTargeting:bool=Get bIsTargeting()
    isWearingBackPackCape:bool=Get isWearingBackblingCape() -> (isWearingBackblingCape:bool)
   ) -> (Pose) {controlRigAssetReference {blueprintRigClass=&/F_MED_IndoorLaceScore_FaceAcc_CtrlRig.F_MED_IndoorLaceScore_FaceAcc_CtrlRig_C}}
   Save cached pose 'controlRigCache'(Pose=@n2) {cacheName=controlRigCache}
   n3 = RigidBody!(ComponentPose=Local To Component(LocalPose=Use cached pose 'controlRigCache'()), Alpha, OverrideWorldGravity:Vector="0,0,-980") -> (Pose) {
    bClampLinearTranslationLimitToRefPose=true
    bOverrideWorldGravity=true
    overridePhysicsAsset=&/F_MED_IndoorLaceScore_FaceAcc_Physics
    simulationSpace=BaseBoneSpace
    alphaBoolBlend {blendOption=Linear}
    baseBoneRef {boneName=neck_02}
    overrideWorldGravity {z=-980}
    simSpaceSettings {
     dampingAlpha=0
     maxAngularAcceleration=9
     maxAngularVelocity=4
     maxLinearAcceleration=1500
     maxLinearVelocity=350
     worldAlpha=1
     externalLinearDragV=<1.1 1.1>
    }
   }
   Save cached pose 'dynamicsCache'(Pose=Component To Local(ComponentPose=@n3)) {cacheName=dynamicsCache}
   n4 = "Constraint - dyn_hair_03_lf (P)"!(ComponentPose=Local To Component(LocalPose=Use cached pose 'dynamicsCache'()), bAlphaBoolEnabled:bool=True, ConstraintWeights_0=1) -> (Pose) {alphaInputType=Bool, boneToModify {boneName=dyn_hair_03_lf}, constraintSetup [{targetBone {boneName=spine_05}}], constraintWeights [1]}
   n5 = "Constraint - dyn_hair_04_lf (P)"!(ComponentPose=@n4, bAlphaBoolEnabled:bool=True, ConstraintWeights_0=1) -> (Pose) {alphaInputType=Bool, boneToModify {boneName=dyn_hair_04_lf}, constraintSetup [{targetBone {boneName=spine_05}}], constraintWeights [1]}
   n6 = "Constraint - dyn_hair_05_lf (P)"!(ComponentPose=@n5, bAlphaBoolEnabled:bool=True, ConstraintWeights_0=1) -> (Pose) {alphaInputType=Bool, boneToModify {boneName=dyn_hair_05_lf}, constraintSetup [{targetBone {boneName=spine_05}}], constraintWeights [1]}
   n7 = "Constraint - dyn_hair_03_cr (P)"!(ComponentPose=@n6, bAlphaBoolEnabled:bool=True, ConstraintWeights_0=1) -> (Pose) {alphaInputType=Bool, boneToModify {boneName=dyn_hair_03_cr}, constraintSetup [{targetBone {boneName=spine_05}}], constraintWeights [1]}
   n8 = "Constraint - dyn_hair_04_cr (P)"!(ComponentPose=@n7, bAlphaBoolEnabled:bool=True, ConstraintWeights_0=1) -> (Pose) {alphaInputType=Bool, boneToModify {boneName=dyn_hair_04_cr}, constraintSetup [{targetBone {boneName=spine_05}}], constraintWeights [1]}
   n9 = "Constraint - dyn_hair_05_cr (P)"!(ComponentPose=@n8, bAlphaBoolEnabled:bool=True, ConstraintWeights_0=1) -> (Pose) {alphaInputType=Bool, boneToModify {boneName=dyn_hair_05_cr}, constraintSetup [{targetBone {boneName=spine_05}}], constraintWeights [1]}
   n10 = "Constraint - dyn_hair_03_rt (P)"!(ComponentPose=@n9, bAlphaBoolEnabled:bool=True, ConstraintWeights_0=1) -> (Pose) {alphaInputType=Bool, boneToModify {boneName=dyn_hair_03_rt}, constraintSetup [{targetBone {boneName=spine_05}}], constraintWeights [1]}
   n11 = "Constraint - dyn_hair_04_rt (P)"!(ComponentPose=@n10, bAlphaBoolEnabled:bool=True, ConstraintWeights_0=1) -> (Pose) {alphaInputType=Bool, boneToModify {boneName=dyn_hair_04_rt}, constraintSetup [{targetBone {boneName=spine_05}}], constraintWeights [1]}
   n12 = "Constraint - dyn_hair_04_rt (P)"(ComponentPose=@n11, bAlphaBoolEnabled=True, ConstraintWeights_0=1) {alphaInputType=Bool, boneToModify {boneName=dyn_hair_04_rt}, constraintSetup [{targetBone {boneName=spine_05}}], constraintWeights [1]}
   n13 = "Transform (Modify) Bone - Bone: dyn_hair_03_lf"!(ComponentPose=@n12, bAlphaBoolEnabled:bool=True, Rotation="-5,0,0", Scale="1,0.5,0.5") -> (Pose) {
    alphaInputType=Bool
    rotationMode=BMM_Additive
    rotationSpace=BCS_BoneSpace
    scaleMode=BMM_Additive
    translationSpace=BCS_BoneSpace
    boneToModify {boneName=dyn_hair_03_lf}
    rotation {pitch=-5}
    scale {y=0.5, z=0.5}
   }
   n14 = "Transform (Modify) Bone - Bone: dyn_hair_03_cr"!(ComponentPose=@n13, bAlphaBoolEnabled:bool=True, Rotation="-5,0,0", Scale="1,1,0.5") -> (Pose) {
    alphaInputType=Bool
    rotationMode=BMM_Additive
    rotationSpace=BCS_BoneSpace
    scaleMode=BMM_Additive
    boneToModify {boneName=dyn_hair_03_cr}
    rotation {pitch=-5}
    scale {z=0.5}
   }
   n15 = "Transform (Modify) Bone - Bone: dyn_hair_03_rt"!(ComponentPose=@n14, bAlphaBoolEnabled:bool=True, Rotation="-5,0,0", Scale="1,1,0.5") -> (Pose) {
    alphaInputType=Bool
    rotationMode=BMM_Additive
    rotationSpace=BCS_BoneSpace
    scaleMode=BMM_Additive
    boneToModify {boneName=dyn_hair_03_rt}
    rotation {pitch=-5}
    scale {z=0.5}
   }
   Save cached pose 'backpackCapeCache'(Pose=Component To Local(ComponentPose=@n15)) {cacheName=backpackCapeCache}
  ]}
 EventGraph {nodes [
   Event OnCharacterCustomizationLoaded()
   > n1 = Cast To PlayerPawn_Athena(Object=TryGetPawnOwner())
   > Set Body_Mesh(Body_Mesh=GetSkeletalMeshForPartType(self=@n1))
   n2 = CheckForBackbling()
   n3 = NOT Boolean(A=@n2.bIsBackblingLong)
   n4 = NOT Boolean(A=@n2.bIsBackblingACape)
   > Set isWearingBackblingCape(isWearingBackblingCape:bool=OR Boolean(A=AND Boolean!(A=@n2, B=@n4, C:bool=@n3) -> (ReturnValue), B=OR Boolean(A=@n2.bIsBackblingACape, B=@n2.bIsBackblingLong))) -> (Output_Get:bool)
   Event BlueprintUpdateAnimation()
   > n5 = "Gravity Override State Function (Parameter Asset) (Nativized)"(
    CharacterPartAnimInstance=Self-Reference()
    GameTimeInSeconds=GetGameTimeInSeconds()
    Parameters=&/F_MED_IndoorLaceScore_FaceAcc_Parameters
   )
   > Set GravityOverride(GravityOverride=@n5)
  ]}
]
variables [
 Body_Mesh {type=SkeletalMeshComponent}
 GravityOverride {type=Vector, default=<0 0 -980>}
 isWearingBackblingCape {type=boolean}
]
